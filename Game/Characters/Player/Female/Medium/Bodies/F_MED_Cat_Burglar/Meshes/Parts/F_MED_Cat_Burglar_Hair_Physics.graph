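# Physics asset for the hair of the female medium Cat Burglar player character, with collision shapes on head, clavicles and upper spine.
# Five hair chains (front and side pairs, one back) each run from a small master box through an upper box to a capsule tip, with limited swing and tight twist.
# Masters attach to the head with free angular motion.
# Free-rotation cross-constraints also link neighboring chains.
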PhysicsAsset /Game/Characters/Player/Female/Medium/Bodies/F_MED_Cat_Burglar/Meshes/Parts/F_MED_Cat_Burglar_Hair_Physics
bodies {
 clavicle_l [{shapeType=Capsule, length=7.75, radius=5.25, center=<6 0 -1>, rotation=<rot -90 -3.7927 3.7927>}]
 clavicle_r [{shapeType=Capsule, length=7.75, radius=5.25, center=<-6 0 1>, rotation=<rot -90 -3.7927 3.7927>}]
 dyn_hairBackHigh_l [{shapeType=Box, extentX=18, extentY=10, extentZ=3.5, center=<6.0487 0.0001 3>}]
 dyn_hairBack_l [
  {shapeType=Capsule, length=10, radius=3.25, center=<8.8099 0 0.8001>, rotation=<rot -90 0 -90.0005>}
  {shapeType=Capsule, length=12, radius=3, center=<3.1103 0 -0.0002>, rotation=<rot 0 0 90.0002>}
 ]
 dyn_hairBack_mstr_l [{shapeType=Box, extentX=1.01, extentY=1.01, extentZ=1.01}]
 dyn_hairFrontHigh_l [{shapeType=Box, extentX=11.5, extentY=10, extentZ=3.5, center=<4 0 4>, rotation=<rot 0 0 40>}]
 dyn_hairFrontHigh_r [{shapeType=Box, extentX=11.5, extentY=10, extentZ=3.5, center=<4 0 0>, rotation=<rot 0 0 -29.9999>}]
 dyn_hairFront_l [{shapeType=Capsule, length=4.75, radius=3.25, center=<2.8375 0 0>, rotation=<rot -90 -26.565 -63.4354>}]
 dyn_hairFront_mstr_l [{shapeType=Box, extentX=1.01, extentY=1.01, extentZ=1.01}]
 dyn_hairFront_mstr_r [{shapeType=Box, extentX=1.01, extentY=1.01, extentZ=1.01}]
 dyn_hairFront_r [{shapeType=Capsule, length=2.25, radius=3.25, center=<2.8375 0 0>, rotation=<rot -90 -26.565 -63.4354>}]
 dyn_hairSideHigh_l [{shapeType=Box, extentX=15, extentY=10, extentZ=3.5, center=<5.1902 0 3.0001>}]
 dyn_hairSideHigh_r [{shapeType=Box, extentX=15, extentY=10, extentZ=3.5, center=<5.1245 0 0>}]
 dyn_hairSide_l [{shapeType=Capsule, length=4.25, radius=3.75, center=<2.8375 0 0>, rotation=<rot -69.9993 89.9988 179.9999>}]
 dyn_hairSide_mstr_l [{shapeType=Box, extentX=1.01, extentY=1.01, extentZ=1.01}]
 dyn_hairSide_mstr_r [{shapeType=Box, extentX=1.01, extentY=1.01, extentZ=1.01}]
 dyn_hairSide_r [{shapeType=Capsule, length=6.5, radius=3.75, center=<2.8375 0 0>, rotation=<rot -69.9996 -89.9992 0>}]
 head [{shapeType=Capsule, length=10.75, radius=4.25, center=<2.8375 0 0>, rotation=<rot -90 -12.9074 22.9077>}]
 spine_05 [
  {shapeType=Capsule, length=13, radius=8, center=<0.6824 -1.8747 -0.0001>, rotation=<rot -90 14.0363 5.9653>}
  {shapeType=Capsule, length=4.25, radius=4.25, center=<17.89 1.8235 0>, rotation=<rot -90 0.0264 -0.0263>}
  {shapeType=Capsule, length=11, radius=9, center=<2.2159 0.9698 0>}
 ]
}
constraints [
 {bone1Name=dyn_hairFront_l, bone2Name=dyn_hairFrontHigh_l, swing1LimitDegrees=30, swing2LimitDegrees=45, twistLimitDegrees=5}
 {bone1Name=dyn_hairBack_l, bone2Name=dyn_hairBackHigh_l, swing1LimitDegrees=45, swing2LimitDegrees=45, twistLimitDegrees=5}
 {bone1Name=dyn_hairBackHigh_l, bone2Name=dyn_hairBack_mstr_l, swing1LimitDegrees=30, swing2LimitDegrees=45, twistLimitDegrees=5}
 {bone1Name=dyn_hairSide_l, bone2Name=dyn_hairSideHigh_l, swing1LimitDegrees=45, swing2LimitDegrees=45, twistLimitDegrees=5}
 {bone1Name=dyn_hairSideHigh_l, bone2Name=dyn_hairSide_mstr_l, swing1LimitDegrees=30, swing2LimitDegrees=45, twistLimitDegrees=5}
 {bone1Name=dyn_hairSide_r, bone2Name=dyn_hairSideHigh_r, swing1LimitDegrees=45, swing2LimitDegrees=45, twistLimitDegrees=5}
 {bone1Name=dyn_hairSideHigh_r, bone2Name=dyn_hairSide_mstr_r, swing1LimitDegrees=30, swing2LimitDegrees=45, twistLimitDegrees=5}
 {
  bone1Name=dyn_hairFront_mstr_l
  bone2Name=head
  swing1LimitDegrees=45
  swing1Motion=Free
  swing2LimitDegrees=45
  swing2Motion=Free
  twistLimitDegrees=45
  twistMotion=Free
 }
 {
  bone1Name=dyn_hairFront_mstr_r
  bone2Name=head
  swing1LimitDegrees=45
  swing1Motion=Free
  swing2LimitDegrees=45
  swing2Motion=Free
  twistLimitDegrees=45
  twistMotion=Free
 }
 {
  bone1Name=dyn_hairSide_mstr_r
  bone2Name=head
  swing1LimitDegrees=45
  swing1Motion=Free
  swing2LimitDegrees=45
  swing2Motion=Free
  twistLimitDegrees=45
  twistMotion=Free
 }
 {
  bone1Name=dyn_hairSide_mstr_l
  bone2Name=head
  swing1LimitDegrees=45
  swing1Motion=Free
  swing2LimitDegrees=45
  swing2Motion=Free
  twistLimitDegrees=45
  twistMotion=Free
 }
 {
  bone1Name=dyn_hairBack_mstr_l
  bone2Name=head
  swing1LimitDegrees=45
  swing1Motion=Free
  swing2LimitDegrees=45
  swing2Motion=Free
  twistLimitDegrees=45
  twistMotion=Free
 }
 {bone1Name=dyn_hairFront_r, bone2Name=dyn_hairFrontHigh_r, swing1LimitDegrees=30, swing2LimitDegrees=45, twistLimitDegrees=5}
 {bone1Name=dyn_hairFrontHigh_l, bone2Name=dyn_hairFront_mstr_l, swing1LimitDegrees=30, swing2LimitDegrees=30, twistLimitDegrees=5}
 {bone1Name=dyn_hairFrontHigh_r, bone2Name=dyn_hairFront_mstr_r, swing1LimitDegrees=30, swing2LimitDegrees=45, twistLimitDegrees=5}
 {
  bone1Name=dyn_hairSideHigh_l
  bone2Name=dyn_hairFrontHigh_l
  swing1LimitDegrees=45
  swing1Motion=Free
  swing2LimitDegrees=45
  swing2Motion=Free
  twistLimitDegrees=45
  twistMotion=Free
 }
 {
  bone1Name=dyn_hairSideHigh_r
  bone2Name=dyn_hairFrontHigh_r
  swing1LimitDegrees=45
  swing1Motion=Free
  swing2LimitDegrees=45
  swing2Motion=Free
  twistLimitDegrees=45
  twistMotion=Free
 }
 {
  bone1Name=dyn_hairSideHigh_r
  bone2Name=dyn_hairSideHigh_l
  swing1LimitDegrees=45
  swing1Motion=Free
  swing2LimitDegrees=45
  swing2Motion=Free
  twistLimitDegrees=45
  twistMotion=Free
 }
 {
  bone1Name=dyn_hairBackHigh_l
  bone2Name=dyn_hairSideHigh_l
  swing1LimitDegrees=45
  swing1Motion=Free
  swing2LimitDegrees=45
  swing2Motion=Free
  twistLimitDegrees=45
  twistMotion=Free
 }
 {
  bone1Name=dyn_hairBackHigh_l
  bone2Name=dyn_hairSideHigh_r
  swing1LimitDegrees=45
  swing1Motion=Free
  swing2LimitDegrees=45
  swing2Motion=Free
  twistLimitDegrees=45
  twistMotion=Free
 }
 {
  bone1Name=dyn_hairBack_l
  bone2Name=dyn_hairSide_r
  swing1LimitDegrees=45
  swing1Motion=Free
  swing2LimitDegrees=45
  swing2Motion=Free
  twistLimitDegrees=45
  twistMotion=Free
 }
 {
  bone1Name=dyn_hairBack_l
  bone2Name=dyn_hairSide_l
  swing1LimitDegrees=45
  swing1Motion=Free
  swing2LimitDegrees=45
  swing2Motion=Free
  twistLimitDegrees=45
  twistMotion=Free
 }
]
